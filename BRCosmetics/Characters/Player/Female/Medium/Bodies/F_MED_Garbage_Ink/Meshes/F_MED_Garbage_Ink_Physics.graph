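# Physics asset for the F_MED_Garbage_Ink female medium body cosmetic, driving secondary motion on its dangling accessory bones.
# Capsules and boxes on the right calf, right upper-arm twist and upper spine bones act as body anchors for the dyn_ bones.
# A right-arm bandana and two right-thigh strings, both tied to the calf, use tight angular limits.
# Three box-shaped tag bodies form a tightly limited chain hanging from spine_05.
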PhysicsAsset /BRCosmetics/Characters/Player/Female/Medium/Bodies/F_MED_Garbage_Ink/Meshes/F_MED_Garbage_Ink_Physics
bodies {
 calf_r [{shapeType=Capsule, length=24.4451, radius=5.2968, center=<21.75 4.4318 0.7755>, rotation=<rot -0.3153 94.6507 -87.0269>}]
 dyn_arm_r_bandana [{shapeType=Capsule, length=4.4893, radius=2.7659, center=<0.2602 -4.3968 -0.7937>, rotation=<rot 0 -171.3287 -72.0317>}]
 dyn_tags [{shapeType=Box, extentX=8.1206, extentY=1.1045, extentZ=6.4511, center=<0 0.706 -3.5047>, rotation=<rot 0 0 -3.0025>}]
 dyn_tags_00 [{shapeType=Box, extentX=9.5815, extentY=0.6582, extentZ=4.4369, center=<0 -0.2191 -0.5669>, rotation=<rot 0 0 -4.8377>}]
 dyn_tags_01 [{shapeType=Box, extentX=8.5023, extentY=0.6582, extentZ=4.4369, center=<0 -0.1627 -0.7485>}]
 dyn_thigh_r_string_l [{shapeType=Capsule, length=2.5416, radius=1.7215, center=<8.2635 0.0249 0.307>, rotation=<rot 0 86.6405 -85.4508>}]
 dyn_thigh_r_string_r [{shapeType=Capsule, length=3.1613, radius=1.6573, center=<7.4467 -0.2578 1.0178>, rotation=<rot -5.769 -90 40.654>}]
 spine_05 [
  {shapeType=Capsule, length=10.5493, radius=12.6887, center=<-9.7302 1.943 0.4134>, rotation=<rot -1.5473 -160.4135 179.887>}
  {shapeType=Box, extentX=7.1035, extentY=3.5718, extentZ=19.0943, center=<2.9102 6.4688 -0.0003>, rotation=<rot 0 -37.1158 0>}
 ]
 upperarm_twist_01_r [{shapeType=Capsule, length=9.5206, radius=5.9537, center=<-0.1158 0.2294 -0.5084>, rotation=<rot 0 -96.7527 -80.1821>}]
}
constraints [
 {bone1Name=dyn_arm_r_bandana, bone2Name=upperarm_twist_01_r, swing1LimitDegrees=10, swing2LimitDegrees=10, twistLimitDegrees=10}
 {bone1Name=dyn_thigh_r_string_r, bone2Name=calf_r, swing1LimitDegrees=10, swing2LimitDegrees=10, twistLimitDegrees=10}
 {bone1Name=dyn_thigh_r_string_l, bone2Name=calf_r, swing1LimitDegrees=10, swing2LimitDegrees=10, twistLimitDegrees=10}
 {bone1Name=dyn_tags, bone2Name=dyn_tags_01, swing1LimitDegrees=5, swing2LimitDegrees=5, twistLimitDegrees=15}
 {bone1Name=dyn_tags_00, bone2Name=spine_05, swing1LimitDegrees=2, swing2LimitDegrees=5, twistLimitDegrees=15}
 {bone1Name=dyn_tags_01, bone2Name=dyn_tags_00, swing1LimitDegrees=2, swing2LimitDegrees=5, twistLimitDegrees=15}
]
