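# Physics asset for the female medium Garbage_Ink character body part, providing capsule bodies for simulated hair.
# A three-bone front hair chain and a separate left temple strand connect to a large head capsule.
# The front chain's root joint swings and twists freely on the head, while the next two links have narrower swing limits and almost no twist.
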PhysicsAsset /BRCosmetics/Characters/Player/Female/Medium/Bodies/F_MED_Garbage_Ink/Meshes/Parts/F_MED_Garbage_Ink_Physics
bodies {
 dyn_hairFront_01 [{shapeType=Capsule, length=0.8153, radius=0.7213, center=<0.4983 0.745 -0.3056>, rotation=<rot 3.1575 -90 19.9528>}]
 dyn_hairFront_02 [{shapeType=Capsule, length=2.4668, radius=1.3798, center=<1.4649 -0.2972 0.2197>, rotation=<rot 0 91.9494 -86.7675>}]
 dyn_hairFront_03 [{shapeType=Capsule, length=2.9006, radius=1.5569, center=<1.9348 0.0067 -0.1746>, rotation=<rot 0 91.935 -89.6691>}]
 dyn_hairTemple_l [{shapeType=Capsule, length=1.7549, radius=0.789, center=<2.007 1.3972 -0.5826>, rotation=<rot 0 -14.3725 -66.8994>}]
 head [{shapeType=Capsule, length=12.5505, radius=10.5598, center=<3.2107 -0.0553 0.3135>, rotation=<rot 0 125.9038 -88.1957>}]
}
constraints [
 {
  bone1Name=dyn_hairFront_01
  bone2Name=head
  swing1LimitDegrees=45
  swing1Motion=Free
  swing2LimitDegrees=45
  swing2Motion=Free
  twistLimitDegrees=45
  twistMotion=Free
 }
 {bone1Name=dyn_hairFront_02, bone2Name=dyn_hairFront_01, swing1LimitDegrees=35, swing2LimitDegrees=35, twistLimitDegrees=1}
 {bone1Name=dyn_hairFront_03, bone2Name=dyn_hairFront_02, swing1LimitDegrees=35, swing2LimitDegrees=35, twistLimitDegrees=1}
 {bone1Name=dyn_hairTemple_l, bone2Name=head, swing1LimitDegrees=45, swing2LimitDegrees=45, twistLimitDegrees=45}
]
