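# Animation Blueprint for the F_MED_Garbage_Ink character part, adding secondary motion, mainly to hair, over a pose copied from the attached parent mesh.
# Four chained Flap nodes drive clamped translation and rotation of dyn_hair_back from head Z rotation.
# AnimDynamics then springs the right and left hair bones, with a neck spherical limit enabled only on the right.
# A final RigidBody pass simulates the override physics asset in spine_04 space; the event graph is empty.
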AnimBlueprint /BRCosmetics/Characters/Player/Female/Medium/Bodies/F_MED_Garbage_Ink/Meshes/Parts/F_MED_Garbage_Ink_AnimBP
target_skeleton=&/F_MED_Garbage_Ink_Skeleton
graphs [
 AnimGraph {nodes [
   n1 = Flap Node - Bone: dyn_hair_back!(
    ComponentPose:Component Space Pose Link Structure=Local To Component(LocalPose=Copy Pose From Mesh() {bUseAttachedParent=true})
    Alpha:float=1
   ) -> (Pose:Component Space Pose Link Structure) {
    is=title FortAnimGraphNode_Flap
    bUseClamp=true
    targetClampMax=5
    targetClampMin=-1
    targetComponent=TranslationY
    alphaBoolBlend {blendOption=Linear}
    sourceDrives [{
      bUseRange=false
      drivingCurve=null
      multiplier=-2
      rangeMax=1
      rangeMin=0
      remappedMax=1
      remappedMin=0
      sourceComponent=RotationZ
      useQuaternion=true
      sourceBone {boneName=head}
     }]
    targetBone {boneName=dyn_hair_back}
   }
   n2 = Flap Node - Bone: dyn_hair_back(ComponentPose=@n1, Alpha=1) {
    bUseClamp=true
    targetClampMax=100
    targetComponent=TranslationX
    alphaBoolBlend {blendOption=Linear}
    sourceDrives [{
      bUseRange=false
      drivingCurve=null
      multiplier=5
      rangeMax=50
      rangeMin=0
      remappedMax=80
      remappedMin=0
      sourceComponent=RotationZ
      useQuaternion=true
      sourceBone {boneName=head}
     }]
    targetBone {boneName=dyn_hair_back}
   }
   n3 = Flap Node - Bone: dyn_hair_back(ComponentPose=@n2, Alpha=1) {
    bUseClamp=true
    targetClampMin=-1
    targetComponent=RotationY
    alphaBoolBlend {blendOption=Linear}
    sourceDrives [{
      bUseRange=false
      drivingCurve=null
      multiplier=-0.25
      rangeMax=50
      rangeMin=0
      remappedMax=50
      remappedMin=0
      sourceComponent=RotationZ
      useQuaternion=true
      sourceBone {boneName=head}
     }]
    targetBone {boneName=dyn_hair_back}
   }
   n4 = Flap Node - Bone: dyn_hair_back(ComponentPose=@n3, Alpha=1) {
    bUseClamp=true
    targetClampMax=10
    targetComponent=RotationZ
    alphaBoolBlend {blendOption=Linear}
    sourceDrives [{
      bUseRange=false
      drivingCurve=null
      multiplier=1.5
      rangeMax=40
      rangeMin=0
      remappedMax=80
      remappedMin=0
      sourceComponent=RotationZ
      useQuaternion=true
      sourceBone {boneName=head}
     }]
    targetBone {boneName=dyn_hair_back}
   }
   n5 = Anim Dynamics - Bone: dyn_hair_r!(ComponentPose=@n4, Alpha=1, AngularSpringConstant:float=4000, ExternalForce="0,0,0") -> (Pose) {
    angularDampingOverride=0.9999
    angularSpringConstant=1000
    bAngularSpring=true
    bGravityOverrideInSimSpace=true
    bOverrideAngularDamping=true
    bUsePlanarLimit=false
    bUseSphericalLimits=true
    gravityScale=0.35
    alphaBoolBlend {blendOption=Linear}
    boundBone {boneName=dyn_hair_r}
    componentAppliedLinearAccClamp=<70 70 70>
    componentLinearVelScale=<300 300 300>
    gravityOverride {z=-980}
    physicsBodyDefinitions [{
      collisionType=CustomSphere
      sphereCollisionRadius=3
      boundBone {boneName=dyn_hair_r}
      boxExtents=<3 20 20>
      constraintSetup {angularLimitsMax=<5 10 20>, angularLimitsMin=<-5 -10 -40>, angularTarget=<1 0 0>}
      localJointOffset=<3 0 0>
     }]
    sphericalLimits [{limitRadius=2.5, drivingBone {boneName=neck_02}, sphereLocalOffset=<0 0 0>}]
   }
   n6 = Anim Dynamics - Bone: dyn_hair_l!(ComponentPose=@n5, Alpha=1, AngularSpringConstant:float=4000, ExternalForce="0,0,0") -> (Pose) {
    angularDampingOverride=1
    angularSpringConstant=1000
    bAngularSpring=true
    bGravityOverrideInSimSpace=true
    bOverrideAngularDamping=true
    bUsePlanarLimit=false
    gravityScale=0.35
    alphaBoolBlend {blendOption=Linear}
    boundBone {boneName=dyn_hair_l}
    componentAppliedLinearAccClamp=<70 70 70>
    componentLinearVelScale=<300 300 300>
    gravityOverride {z=-980}
    physicsBodyDefinitions [{
      collisionType=CustomSphere
      sphereCollisionRadius=3
      boundBone {boneName=dyn_hair_l}
      boxExtents=<3 20 20>
      constraintSetup {angularLimitsMax=<0 10 20>, angularLimitsMin=<0 -10 -65>, angularTarget=<0 0 0>}
      localJointOffset=<2 0 0>
     }]
    sphericalLimits [{limitRadius=5, drivingBone {boneName=neck_02}, sphereLocalOffset=<0 0 0>}]
   }
   n7 = RigidBody(ComponentPose=@n6) {
    bOverrideWorldGravity=true
    overridePhysicsAsset=&/F_MED_Garbage_Ink_Physics
    simulationSpace=BaseBoneSpace
    alphaBoolBlend {blendOption=Linear}
    baseBoneRef {boneName=spine_04}
    componentAppliedLinearAccClamp=<100 100 100>
    componentLinearAccScale=<0.6 0.6 0.6>
    componentLinearVelScale=<0.5 0.5 0.5>
    overrideWorldGravity {z=-980}
   }
   Output Pose(Result=Component To Local(ComponentPose=@n7)) {layerGroup=None}
  ]}
 EventGraph {nodes [Event BlueprintUpdateAnimation()]}
]
variables []
